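# Animation Blueprint for the Embers Male pickaxe skeleton that plays the weapon's looping chain idle animation.
# A single-state machine plays the Embers_PickAxe_Idle_CHAIN_R sequence and feeds DefaultSlot into the output pose, so montages can override it.
# The event graph holds only unconnected update and pawn-owner nodes, with no logic or variables.
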AnimBlueprint /BRCosmetics/Weapons/FORT_Melee/Pickaxe_Embers_Male/Meshes/Embers_Male_PickAxe_Skeleton_AnimBlueprint
target_skeleton=/Game/Weapons/FORT_Melee/Pickaxe_Embers_Male/Meshes/Embers_Male_PickAxe_Skeleton
graphs [
 AnimGraph {nodes [
   AnimGraphNode_StateMachine_0 = New State Machine() {is=title AnimGraphNode_StateMachine}
   Output Pose(Result=Slot 'DefaultSlot'(Source=@AnimGraphNode_StateMachine_0)) {layerGroup=None}
  ]}
 New State Machine {entry=Idle, path=AnimGraph.AnimGraphNode_StateMachine_0.New State Machine, ids=id, nodes [AnimStateNode_0 = Idle()]}
 Idle {
  path=AnimGraph.AnimGraphNode_StateMachine_0.New State Machine.AnimStateNode_0.Idle
  nodes [Output Animation Pose(Result=Sequence Player 'Embers_PickAxe_Idle_CHAIN_R'() {sequence=/BRCosmetics/Animation/Game/MainPlayer/Combat/Melee/DualWield/Embers/EmbersPickaxeAnims/Embers_PickAxe_Idle_CHAIN_R})]
 }
 EventGraph {nodes [
   TryGetPawnOwner()
   Event BlueprintUpdateAnimation()
  ]}
]
variables []
